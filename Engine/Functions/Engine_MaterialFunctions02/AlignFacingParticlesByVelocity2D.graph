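# Material function, exposed to the library, that rotates particle sprites around their centers to follow their on-screen direction of travel.
# ParticleDirection is transformed from world to view space, and its 2D components become an angle via VectorToRadialValue.
# That angle drives RotateAboutAxis on world position, using the world vertex normal as axis and particle position as pivot.
# Additional camera-direction and orthogonal-basis chains are present but do not feed the output.
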MaterialFunction /Engine/Functions/Engine_MaterialFunctions02/AlignFacingParticlesByVelocity2D
graph {
 properties {bExposeToLibrary=true}
 nodes [
  n1 = ComponentMask(None=Transform(None=ParticleDirection()) {transformSourceType=TRANSFORMSOURCE_World, transformType=TRANSFORM_View}) {b=true, g=true}
  FunctionOutput(None=RotateAboutAxis(
    NormalizedRotationAxis=VertexNormalWS()
    RotationAngle=MaterialFunctionCall(Vector or UVs=@n1) {materialFunction=&VectorToRadialValue}.Vector Converted to Angle
    PivotPoint=ParticlePositionWS()
    Position=WorldPosition().XYZ
   ))
  n2 = Multiply(A=ComponentMask(None=MaterialFunctionCall() {materialFunction=&CameraDirectionVector}.Result) {r=true}) {constB=2}
  Normalize(VectorInput=Multiply(
    A=MaterialFunctionCall() {materialFunction=&CameraDirectionVector}.Result
    B=Clamp(None=@n2) {minDefault=-1}
   ) {constB=-1})
  n3 = MaterialFunctionCall(Float2=TextureCoordinate()) {materialFunction=&BreakOutFloat2Components}
  n4 = OneMinus(None=@n3.G)
  n5 = AppendVector(A=Multiply(A=AppendVector(A=@n3.R, B=@n4), B=ParticleSize()), B=Constant())
  n6 = ParticleDirection()
  n7 = MaterialFunctionCall() {materialFunction=&CameraDirectionVector}
  n8 = MaterialFunctionCall!(Vector1=CrossProduct(A=@n6, B=@n7.Result), Vector2=@n6, Vector3=@n7.Result) {materialFunction=/Engine/Functions/Engine_MaterialFunctions02/Math/MakeVectorsOrthogonal}
  Subtract(
   A=Add(
    A=MaterialFunctionCall(VectorToTransform=@n5, BasisX=@n8.Vector1, BasisY=@n8.Vector1, BasisZ=@n8.Vector1) {desc=world orientation, materialFunction=&Transform3x3Matrix}.Transformed Vector
    B=ParticlePositionWS()
   )
   B=WorldPosition().XYZ
  )
 ]
}
