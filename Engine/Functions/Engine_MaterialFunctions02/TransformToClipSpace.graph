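# Material function, exposed to the library, that projects a world position into clip space and screen UVs.
# A Custom HLSL node multiplies the position, defaulting to world position without shader offsets, by WorldToClip, then XY is divided by W.
# The -1 to 1 result is remapped with flipped Y to 0-1 screen UVs.
# Those are scaled by view size, pixel-offset and multiplied by scene texel size, giving scene texture buffer UVs.
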MaterialFunction /Engine/Functions/Engine_MaterialFunctions02/TransformToClipSpace
graph {
 properties {bExposeToLibrary=true, libraryCategoriesText [Misc,Math]}
 nodes [
  MaterialExpressionFunctionInput_15 = FunctionInput(Preview=WorldPosition() {worldPositionShaderOffset=WPT_ExcludeAllShaderOffsets}.XYZ) {bUsePreviewValueAsDefault=true, inputName=World Position}
  n1 = Custom!(In=@MaterialExpressionFunctionInput_15) {
   code="mul(float4(In.xyz,1),LWCToFloat(ResolvedView.WorldToClip))"
   desc="\n*** Original source before expression upgrade ***\nmul(float4(In.xyz,1),View.WorldToClip)"
   outputType=CMOT_Float4
   inputs [{inputName=In, input {expression=&self:MaterialExpressionFunctionInput_15}}]
  }
  n2 = ComponentMask(None=@n1) {g=true, r=true}
  n3 = ComponentMask(None=@n1) {a=true}
  n4 = MaterialFunctionCall(Float2=FunctionOutput(None=Divide(A=@n2, B=@n3)) {description=A clip space -1 to 1 value., outputName=Clip Space XY, sortPriority=3}) {materialFunction=&BreakOutFloat2Components}
  n5 = Add(A=Multiply(A=@n4.R) {constB=0.5}) {constB=0.5}
  n6 = Add(A=Multiply(A=@n4.G) {constB=-0.5}) {constB=0.5}
  n7 = Multiply(
   A=FunctionOutput(None=AppendVector(A=@n5, B=@n6)) {description="0-1 (if the object is visible).", outputName=Screen aligned 0-1 UV, sortPriority=1}
   B=ViewSize()
  ) {desc=0-1 per pixel}
  n8 = FunctionInput() {
   bUsePreviewValueAsDefault=true
   inputName=Buffer UV Pixel Offset
   inputType=FunctionInput_Vector2
   sortPriority=10
   previewValue=<0.5 0.5 0 1>
  }
  FunctionOutput(None=Multiply(A=Add(A=@n7, B=@n8) {constB=0.5}, B=SceneTexelSize() {desc=1 pixel}) {desc=uv in buffer size}) {desc=Buffer UVs for Scene Textures, outputName=Buffer UVs for Scene Textures}
 ]
}
