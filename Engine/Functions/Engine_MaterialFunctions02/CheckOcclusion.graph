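# Material function for translucent materials that tests whether a world position, actor position by default, is hidden behind opaque geometry.
# Scene depth sampled at the position's projected screen UV minus the position's own depth, divided by a falloff and clamped, gives a visibility value.
# Clip-space XY builds a mask that is zero offscreen and fades near screen edges using sphere masks.
# Outputs provide both terms separately and multiplied together.
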MaterialFunction /Engine/Functions/Engine_MaterialFunctions02/CheckOcclusion
graph {
 properties {
  description=Checks to see if a world position is occluded by an opaque object. For use with translucent materials.
  libraryCategoriesText [Misc,LensFlares]
 }
 nodes [
  n1 = FunctionInput(Preview=ActorPositionWS()) {bUsePreviewValueAsDefault=true, inputName=Position To check for Occlusion}
  n2 = MaterialFunctionCall(World Position=@n1) {materialFunction=&TransformToClipSpace}
  n3 = Abs(None=@n2.Clip Space XY)
  n4 = MaterialFunctionCall(f2=Floor(None=@n3)) {materialFunction=&AddComponents}
  n5 = MaterialFunctionCall(Float2=@n3) {materialFunction=&BreakOutFloat2Components}
  n6 = Constant() {r=1}
  n7 = FunctionInput() {bUsePreviewValueAsDefault=true, inputName=Offscreen Falloff Range, inputType=FunctionInput_Scalar, sortPriority=1, previewValue=<0.05 0 0 1>}
  n8 = Add(
   A=SphereMask(A=@n5.R, B=@n6, Radius=@n7) {attenuationRadius=0.05, hardnessPercent=0}
   B=SphereMask(A=@n5.G, B=@n6, Radius=@n7) {attenuationRadius=0.05, hardnessPercent=0}
  )
  n9 = Multiply(A=OneMinus(None=Clamp(None=@n4.f2)), B=OneMinus(None=Clamp(None=@n8)))
  FunctionOutput(None=@n9) {outputName=Screen Edge, sortPriority=2}
  n10 = SceneDepth(UVs=@n2.Buffer UVs for Scene Textures)
  n11 = MaterialFunctionCall!(World Position=@n1) {materialFunction=/Engine/Functions/Engine_MaterialFunctions02/Utility/DepthFromWorldPosition}
  n12 = FunctionInput() {bUsePreviewValueAsDefault=true, inputName=Depth Occlusion Falloff, inputType=FunctionInput_Scalar, sortPriority=1, previewValue=<10 0 0 1>}
  n13 = Clamp(None=Divide(A=Subtract(A=@n10.R, B=@n11.Result), B=@n12) {constB=30, desc=light radius})
  FunctionOutput(None=@n13) {outputName=Occlusion, sortPriority=1}
  FunctionOutput(None=Multiply(A=@n13, B=@n9)) {outputName=Screen Edge And Occlusion}
  Transform() {transformSourceType=TRANSFORMSOURCE_View}
 ]
}
