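# Material function that provides a tiled world position for mapping textures to the world without precision artifacts at large distances.
# It adds camera-relative world position divided by TilingSize to the fractional part of camera position divided by TilingSize, keeping coordinates small.
# Dot products with U and V axis vectors turn that position into UVs that sample the input texture.
# It outputs both the position and the texture sample.
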MaterialFunction /Engine/Functions/Engine_MaterialFunctions02/HighPrecisionWorldPosTextureSampling
graph {
 properties {
  bExposeToLibrary=true
  description=Provides a world position that can be used to map textures to the world without low precision artifacts at large distances.
  libraryCategoriesText [Misc,Texturing]
 }
 nodes [
  n1 = FunctionInput() {
   bUsePreviewValueAsDefault=true
   description=Enter a single scalar value to scale all of the coordinates at once.
   inputName=TilingSize
   previewValue=<128 128 128 1>
  }
  n2 = CameraPositionWS()
  n3 = Add(
   A=Divide(A=WorldPosition() {worldPositionShaderOffset=WPT_CameraRelative}.XYZ, B=@n1) {constB=128}
   B=Frac(None=Divide(A=@n2, B=@n1) {constB=128})
  )
  FunctionOutput(None=@n3) {outputName=High Precision World Position, sortPriority=1}
  n4 = FunctionInput() {
   bUsePreviewValueAsDefault=true
   description="Enter 1,0,0 to make the first uv channel X or 0,1,0 to make it Y or 0,01 to make it z"
   inputName=Texture Coord U Vector
   sortPriority=1
   previewValue=<1 0 0 1>
  }
  n5 = FunctionInput() {
   bUsePreviewValueAsDefault=true
   description="Enter 1,0,0 to make the second uv channel X or 0,1,0 to make it Y or 0,01 to make it z"
   inputName=Texture Coord V Vector
   sortPriority=2
   previewValue=<0 1 0 1>
  }
  n6 = FunctionInput(Preview=TextureObject()) {bUsePreviewValueAsDefault=true, inputName=Texture Object, inputType=FunctionInput_Texture2D, sortPriority=-1, previewValue=<2048 0 0 1>}
  n7 = TextureSample(UVs=AppendVector(A=DotProduct(A=@n3, B=@n4), B=DotProduct(A=@n3, B=@n5)), Tex=@n6)
  FunctionOutput(None=AppendVector(A=@n7.RGB, B=@n7.RGB)) {outputName="Texture (F4)", sortPriority=-1}
 ]
}
