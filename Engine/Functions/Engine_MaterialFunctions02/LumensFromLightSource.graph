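# Material function, exposed to the library, that returns the per-pixel brightness of a spherical light at a given radius.
# It multiplies Light Color, Light Source Lumens and a constant 16, then divides by the squared Light Radius times Pi, taken from the Pi function.
# All three inputs use their preview values as defaults, so the function works with any of them unconnected.
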MaterialFunction /Engine/Functions/Engine_MaterialFunctions02/LumensFromLightSource
graph {
 properties {bExposeToLibrary=true, description=Returns the per pixel brightness of a spherical light at a given radius.}
 nodes [
  n1 = Multiply(A=Constant() {r=16}, B=FunctionInput() {bUsePreviewValueAsDefault=true, inputName=Light Color, previewValue=<1 1 1 1>})
  n2 = FunctionInput() {bUsePreviewValueAsDefault=true, inputName=Light Source Lumens, inputType=FunctionInput_Scalar, previewValue=<500 1 1 1>}
  n3 = FunctionInput() {bUsePreviewValueAsDefault=true, inputName=Light Radius, inputType=FunctionInput_Scalar, previewValue=<20 1 1 1>}
  n4 = Multiply(A=@n3, B=@n3)
  n5 = MaterialFunctionCall() {materialFunction=&Pi}
  FunctionOutput(None=Divide(A=Multiply(A=@n1, B=@n2), B=Multiply(A=@n4, B=@n5.Result)))
 ]
}
