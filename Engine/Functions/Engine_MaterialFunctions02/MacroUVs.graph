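# MacroUVs is a material function in the engine's Engine_MaterialFunctions02 library, named for macro-scale UV coordinates.
# Its expression graph is empty here, with no inputs, outputs or math nodes, so the file defines no UV computation.
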MaterialFunction /Engine/Functions/Engine_MaterialFunctions02/MacroUVs
graph {expressions {}}
