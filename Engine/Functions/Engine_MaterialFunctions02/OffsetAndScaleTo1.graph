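# Material function, exposed to the library, that rescales a scalar for an offset, described as bringing offset values back into the 0-1 range.
# It multiplies Original Value by the reciprocal of Offset Amount plus the Add node's default constant of one, then outputs the red channel.
# The graph never adds the offset to Original Value, so only the scaling step happens here, despite the description.
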MaterialFunction /Engine/Functions/Engine_MaterialFunctions02/OffsetAndScaleTo1
graph {
 properties {
  bExposeToLibrary=true
  description="The value entered in the original value input is offset by the offset value and the values are then scaled back into the 0-1 range. "
 }
 nodes [
  n1 = FunctionInput() {bUsePreviewValueAsDefault=true, inputName=Original Value, inputType=FunctionInput_Scalar}
  n2 = Divide(
   A=Constant() {r=1}
   B=Add(A=FunctionInput() {bUsePreviewValueAsDefault=true, inputName=Offset Amount, inputType=FunctionInput_Scalar, sortPriority=1})
  )
  FunctionOutput(None=ComponentMask(None=Multiply(A=@n1, B=@n2)) {r=true})
 ]
}
