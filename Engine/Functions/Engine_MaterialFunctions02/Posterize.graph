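# Material function that posterizes a color input, quantizing its brightness into a configurable number of value steps.
# A dot product sums the Diffuse RGB channels, and that sum is multiplied by the step count, rounded up with Ceil and divided back down.
# The color divided by its original sum is scaled by this stepped value, so channel ratios stay intact while intensity bands.
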MaterialFunction /Engine/Functions/Engine_MaterialFunctions02/Posterize
graph {nodes [
  n1 = TextureCoordinate()
  n2 = FunctionInput(Preview=AppendVector(A=@n1, B=ComponentMask(None=@n1) {g=true})) {inputName=Diffuse, previewValue=<0.5 0 0 1>}
  n3 = DotProduct(A=Constant4Vector() {constant=<rgba 1 1 1 0>}.RGBA, B=AppendVector(A=@n2, B=Constant() {r=1}))
  n4 = FunctionInput() {inputName=Number of value steps, inputType=FunctionInput_Scalar, previewValue=<5 0 0 1>}
  n5 = Ceil(None=Multiply(A=@n3, B=@n4) {constB=30})
  FunctionOutput(None=Multiply(A=Divide(A=@n2, B=@n3), B=Divide(A=@n5, B=@n4)))
 ]}
