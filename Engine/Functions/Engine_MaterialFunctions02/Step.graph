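# Step is a material function that compares two scalar inputs, Low and High, to produce a hard binary threshold.
# An If node outputs one when Low is below High and zero when Low is above it.
# Both inputs fall back to preview values when unconnected: a midpoint constant for Low and the green channel of the texture coordinates for High.
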MaterialFunction /Engine/Functions/Engine_MaterialFunctions02/Step
graph {nodes [
  n1 = Constant() {r=0.5}
  n2 = FunctionInput(Preview=ComponentMask(None=TextureCoordinate()) {g=true}) {bUsePreviewValueAsDefault=true, description=High threshold point for comparision, inputName=High, inputType=FunctionInput_Scalar}
  FunctionOutput(None=If(
    A=FunctionInput(Preview=@n1) {bUsePreviewValueAsDefault=true, description=Low threshold value, inputName=Low, inputType=FunctionInput_Scalar}
    B=@n2
    "A > B"=Constant()
    "A < B"=Constant() {r=1}
   ))
 ]}
