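# Material function that remaps a 0-1 scalar through new black, middle and white values around an adjustable middle point, exposed under Image Adjustment.
# Clamped ratios drive two lerps, black-to-middle and white-to-middle, and the floored lower ratio selects between them without a branch.
# Static switches optionally apply a power curve to both segments, with an invert option that flips before and after the power.
# Two unconnected dashed-name bool inputs act as separators in the input list.
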MaterialFunction /Engine/Functions/Engine_MaterialFunctions02/3PointLevels
graph {
 properties {
  bExposeToLibrary=true
  description=Remaps 0-1 values by linearly interpolating through 3 new values.
  libraryCategoriesText [Misc, Image Adjustment]
 }
 nodes [
  FunctionInput() {
   bUsePreviewValueAsDefault=true
   description=Pick a value from the input image that will get replaced with the value from Middle Value.
   inputName=---------------
   inputType=FunctionInput_StaticBool
   sortPriority=1
   previewValue=<1 0 0 1>
  }
  FunctionInput() {
   bUsePreviewValueAsDefault=true
   description=Pick a value from the input image that will get replaced with the value from Middle Value.
   inputName="--------------- "
   inputType=FunctionInput_StaticBool
   sortPriority=6
   previewValue=<1 0 0 1>
  }
  n1 = FunctionInput() {
   bUsePreviewValueAsDefault=true
   description=Set the new value for the previous value of 0.
   inputName=New Black Value
   inputType=FunctionInput_Scalar
   sortPriority=1
  }
  n2 = FunctionInput() {
   bUsePreviewValueAsDefault=true
   description=Replaces the old value that matches the Middle Point input value.
   inputName=New Middle Value
   inputType=FunctionInput_Scalar
   sortPriority=2
   previewValue=<0.5 0 0 1>
  }
  n3 = FunctionInput(Preview=ComponentMask(None=TextureCoordinate()) {r=true}) {
   bUsePreviewValueAsDefault=true
   description=Pick a value from the input image that will get replaced with the value from Middle Value.
   inputName=Texture
   inputType=FunctionInput_Scalar
   previewValue=<1 0 0 1>
  }
  n4 = FunctionInput() {
   bUsePreviewValueAsDefault=true
   description=Pick a value from the input image that will get replaced with the value from Middle Value.
   inputName=Middle Point
   inputType=FunctionInput_Scalar
   sortPriority=5
   previewValue=<0.5 0 0 1>
  }
  n5 = Clamp(None=Divide(A=@n3, B=@n4))
  n6 = FunctionInput(Preview=StaticBool()) {
   bUsePreviewValueAsDefault=true
   description=Invert the power curve.
   inputName=Invert Interpolation Power
   inputType=FunctionInput_StaticBool
   sortPriority=9
   previewValue=<2 0 0 1>
  }
  n7 = FunctionInput() {
   bUsePreviewValueAsDefault=true
   description=A power node applied to the linear interpolation. This defines the curve of the interpolation.
   inputName=Interpolation Power
   inputType=FunctionInput_Scalar
   sortPriority=8
   previewValue=<2 0 0 1>
  }
  n8 = Power(Base=StaticSwitch(True=OneMinus(None=@n5), False=@n5, Value=@n6), Exp=@n7)
  n9 = FunctionInput(Preview=StaticBool()) {
   bUsePreviewValueAsDefault=true
   description=Pick a value from the input image that will get replaced with the value from Middle Value.
   inputName=Define Interpolation Curve
   inputType=FunctionInput_StaticBool
   sortPriority=7
   previewValue=<1 0 0 1>
  }
  n10 = StaticSwitch(True=StaticSwitch(True=OneMinus(None=@n8), False=@n8, Value=@n6), False=@n5, Value=@n9)
  n11 = FunctionInput() {
   bUsePreviewValueAsDefault=true
   description=Set the new value for the previous value of 1.
   inputName=New White Value
   inputType=FunctionInput_Scalar
   sortPriority=3
   previewValue=<1 0 0 1>
  }
  n12 = Clamp(None=Divide(A=OneMinus(None=@n3), B=OneMinus(None=@n4)))
  n13 = Power(Base=StaticSwitch(True=OneMinus(None=@n12), False=@n12, Value=@n6), Exp=@n7)
  n14 = StaticSwitch(True=StaticSwitch(True=OneMinus(None=@n13), False=@n13, Value=@n6), False=@n12, Value=@n9)
  FunctionOutput(None=LinearInterpolate(
    A=LinearInterpolate(A=@n1, B=@n2, Alpha=@n10) {desc=FirstLerp}
    B=LinearInterpolate(A=@n11, B=@n2, Alpha=@n14) {desc=2ndLerp}
    Alpha=Floor(None=@n10) {desc=LerpMask}
   ))
 ]
}
