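# Material function exposed in the library as a configurable Fresnel, with more exposed properties and a cheap alternative falloff.
# It dots a normal (default vertex normal) with the camera vector, clamps by default, then applies one-minus unless inverted.
# Static switches select the falloff at compile time: Power by default, or a cheaper lerp between dark and bright levels for contrast.
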MaterialFunction /Engine/Functions/Engine_MaterialFunctions02/Fresnel_Function
graph {
 properties {
  bExposeToLibrary=true
  description="Fresnel with more properties exposed, less instructions and a cheap alternative falloff."
  libraryCategoriesText [Misc,VectorOps]
 }
 nodes [
  n1 = FunctionInput() {
   bUsePreviewValueAsDefault=true
   description=Black level
   inputName=Cheap contrast dark
   inputType=FunctionInput_Scalar
   sortPriority=11
   previewValue=<0 0 1 1>
  }
  n2 = FunctionInput() {
   bUsePreviewValueAsDefault=true
   description=Bright level
   inputName=Cheap contrast bright
   inputType=FunctionInput_Scalar
   sortPriority=12
   previewValue=<1 0 1 1>
  }
  n3 = DotProduct(
   A=FunctionInput(Preview=VertexNormalWS()) {bUsePreviewValueAsDefault=true, inputName=Normal Vector, previewValue=<0 0 1 1>}
   B=FunctionInput(Preview=CameraVectorWS()) {bUsePreviewValueAsDefault=true, description="Default Value = Camera Vector", inputName=Camera Vector, sortPriority=1}
  )
  n4 = FunctionInput(Preview=StaticBool() {value=true}) {
   bUsePreviewValueAsDefault=true
   description="Default Value = True"
   inputName=Clamp Fresnel Dot Product
   inputType=FunctionInput_StaticBool
   sortPriority=100
   previewValue=<0 0 1 1>
  }
  n5 = StaticSwitch(True=Clamp(None=@n3), False=@n3, Value=@n4)
  n6 = FunctionInput(Preview=StaticBool()) {
   bUsePreviewValueAsDefault=true
   description="Default Value = False"
   inputName=Invert Fresnel
   inputType=FunctionInput_StaticBool
   sortPriority=4
   previewValue=<0 0 1 1>
  }
  n7 = StaticSwitch(True=@n5, False=OneMinus(None=@n5), Value=@n6)
  n8 = FunctionInput() {
   bUsePreviewValueAsDefault=true
   description=This is used when cheap contrast is set to false.
   inputName=Power
   inputType=FunctionInput_Scalar
   sortPriority=9
   previewValue=<1 0 1 1>
  }
  n9 = FunctionInput(Preview=StaticBool()) {
   bUsePreviewValueAsDefault=true
   description="Default Value = False"
   inputName=Use Cheap contrast
   inputType=FunctionInput_StaticBool
   sortPriority=10
   previewValue=<0 0 1 1>
  }
  FunctionOutput(None=StaticSwitch(True=LinearInterpolate(A=@n1, B=@n2, Alpha=@n7), False=Power(Base=@n7, Exp=@n8), Value=@n9))
 ]
}
