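# HueShift is a library material function under Misc and Image Adjustment that adjusts the hue of an input color.
# It rotates the Texture input as a vector about the normalized (1,1,1) gray axis through the origin, using Hue Shift Percentage as the rotation angle.
# RotateAboutAxis yields an offset, which is added back to the original color to produce the output.
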MaterialFunction /Engine/Functions/Engine_MaterialFunctions02/HueShift
graph {
 properties {bExposeToLibrary=true, description=Adjusts the hue of an image., libraryCategoriesText [Misc, Image Adjustment]}
 nodes [
  n1 = Normalize(VectorInput=Constant3Vector() {constant=<rgba 1 1 1 0>})
  n2 = FunctionInput() {bUsePreviewValueAsDefault=true, inputName=Hue Shift Percentage, inputType=FunctionInput_Scalar}
  n3 = Constant()
  n4 = FunctionInput(Preview=Constant3Vector() {constant=<rgba 0 1 0 0>}) {bUsePreviewValueAsDefault=true, inputName=Texture}
  FunctionOutput(None=Add(A=RotateAboutAxis(NormalizedRotationAxis=@n1, RotationAngle=@n2, PivotPoint=@n3, Position=@n4), B=@n4))
 ]
}
