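# Library material function that blends three or four inputs in sequence from a single 0-1 scalar alpha, with separate outputs for each count.
# Alpha is multiplied by the number of segments, then offset by successive subtractions and clamped to drive chained linear interpolations, giving each neighboring pair an equal share of the range.
# Preview values supply default inputs, with fractional time animating the alpha preview.
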MaterialFunction /Engine/Functions/Engine_MaterialFunctions02/Lerp_Multiple_Float3
graph {
 properties {bExposeToLibrary=true, description=Lerps multiple inputs based on a single value.}
 nodes [
  n1 = FunctionInput() {bUsePreviewValueAsDefault=true, inputName=1, previewValue=<1 0 0 1>}
  n2 = FunctionInput() {bUsePreviewValueAsDefault=true, inputName=2, sortPriority=1, previewValue=<0 1 0 1>}
  n3 = FunctionInput(Preview=MaterialFunctionCall() {materialFunction=&TimeWithSpeedVariable}.Frac Time) {bUsePreviewValueAsDefault=true, description=0-1 lerp alpha, inputName=A, inputType=FunctionInput_Scalar, sortPriority=4}
  n4 = Multiply(A=@n3) {constB=2}
  n5 = Clamp(None=@n4)
  n6 = FunctionInput() {bUsePreviewValueAsDefault=true, inputName=3, sortPriority=2, previewValue=<0 0 1 1>}
  n7 = Subtract(A=@n4)
  FunctionOutput(None=LinearInterpolate(A=LinearInterpolate(A=@n1, B=@n2, Alpha=@n5), B=@n6, Alpha=Clamp(None=@n7))) {outputName=Lerp 3 Inputs}
  n8 = Multiply(A=@n3) {constB=3}
  n9 = LinearInterpolate(A=@n1, B=@n2, Alpha=Clamp(None=@n8))
  n10 = Subtract(A=@n8)
  n11 = Clamp(None=@n10)
  n12 = Subtract(A=@n10)
  FunctionOutput(None=LinearInterpolate(
    A=LinearInterpolate(A=@n9, B=@n6, Alpha=@n11)
    B=FunctionInput() {bUsePreviewValueAsDefault=true, inputName=4, sortPriority=3}
    Alpha=Clamp(None=@n12)
   )) {outputName=Lerp 4 Inputs, sortPriority=1}
 ]
}
