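# MatLayerBlend_BreakOpacity is a library-exposed material function that extracts the Opacity channel from a set of material attributes.
# Its "Mat attributes" input feeds a BreakMaterialAttributes node, whose Opacity pin goes straight to the function's single output.
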MaterialFunction /Engine/Functions/Engine_MaterialFunctions02/MatLayerBlend_BreakOpacity
graph {
 properties {bExposeToLibrary=true}
 nodes [FunctionOutput(None=BreakMaterialAttributes!(Attr=FunctionInput() {inputName=Mat attributes, inputType=FunctionInput_MaterialAttributes}).Opacity)]
}
