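# ObjectLocalBounds is an engine material function that exposes the rendered primitive's local-space bounding box to materials.
# Two Custom HLSL nodes read LocalObjectBoundsMin and LocalObjectBoundsMax from the primitive data for the minimum and maximum outputs.
# A third output, Local Bounds Size, subtracts the minimum from the maximum to give the box extents along each axis.
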MaterialFunction /Engine/Functions/Engine_MaterialFunctions02/ObjectLocalBounds
graph {nodes [
  n1 = Custom() {code="GetPrimitiveData(Parameters).LocalObjectBoundsMin.xyz", description=Local Bounds Minimum, inputs []}
  FunctionOutput(None=@n1) {outputName=Local Bounds Minimum}
  n2 = Custom() {code="GetPrimitiveData(Parameters).LocalObjectBoundsMax.xyz", description=Local Bounds Max, inputs []}
  FunctionOutput(None=@n2) {outputName=Local Bounds Max, sortPriority=1}
  FunctionOutput(None=Subtract(A=@n2, B=@n1)) {outputName=Local Bounds Size, sortPriority=2}
 ]}
