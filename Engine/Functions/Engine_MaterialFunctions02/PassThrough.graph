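# PassThrough is an engine material function, exposed to the library, that returns its inputs unchanged as a hub for tidier material graph wiring.
# It has four inputs, scalar S and vectors V2, V3 and V4, each wired directly to an output of the same name with no math between.
# Every input uses its preview value as the default, so any of them can be left unconnected.
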MaterialFunction /Engine/Functions/Engine_MaterialFunctions02/PassThrough
graph {
 properties {
  bExposeToLibrary=true
  description="This node simply passes any inserted values through itself. Suggested uses include, using this node to form a closer connection point for nodes on opposite sides of the graph or bufferering math operations and the nodes they connect to with this connection hub node. "
 }
 nodes [
  FunctionOutput(None=FunctionInput() {bUsePreviewValueAsDefault=true, description=Insert a value to be passed through the node., inputName=V3, sortPriority=2}) {outputName=V3, sortPriority=2}
  n1 = FunctionInput() {
   bUsePreviewValueAsDefault=true
   description=Insert a value to be passed through the node.
   inputName=V2
   inputType=FunctionInput_Vector2
   sortPriority=1
  }
  FunctionOutput(None=@n1) {outputName=V2, sortPriority=1}
  n2 = FunctionInput() {bUsePreviewValueAsDefault=true, description=Insert a value to be passed through the node., inputName=S, inputType=FunctionInput_Scalar}
  FunctionOutput(None=@n2) {outputName=S}
  n3 = FunctionInput() {
   bUsePreviewValueAsDefault=true
   description=Insert a value to be passed through the node.
   inputName=V4
   inputType=FunctionInput_Vector4
   sortPriority=3
  }
  FunctionOutput(None=@n3) {outputName=V4, sortPriority=3}
 ]
}
