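# Material function that analytically intersects a ray with a sphere, exposed to the library under Misc and Math.
# The ray starts at the world position, defaults to pointing toward the camera, and feeds a quadratic solve against sphere position and radius.
# The far root yields depth, intersection position, surface normal and center-relative vector, while the discriminant's sign gives the hit flag.
# A separate density output remaps radius-normalized distance from the center.
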MaterialFunction /Engine/Functions/Engine_MaterialFunctions02/RayTracedSphere
graph {
 properties {bExposeToLibrary=true, libraryCategoriesText [Misc,Math]}
 nodes [
  n1 = FunctionInput(Preview=WorldPosition().XYZ) {bUsePreviewValueAsDefault=true, inputName=World position, sortPriority=6}
  n2 = FunctionInput() {bUsePreviewValueAsDefault=true, inputName=Sphere Position, sortPriority=-1, previewValue=<0 0 0 256>}
  n3 = Subtract(A=@n1, B=@n2)
  n4 = DotProduct(A=@n3, B=@n3)
  n5 = FunctionInput() {bUsePreviewValueAsDefault=true, inputName=Sphere Radius, inputType=FunctionInput_Scalar, previewValue=<16 0 0 16>}
  n6 = Clamp(None=Divide(A=SquareRoot(None=@n4), B=@n5))
  n7 = Subtract(A=@n6) {constB=0.6}
  FunctionOutput(None=OneMinus(None=Multiply(A=@n7) {constB=2.5})) {outputName=0-1 Sphere density, sortPriority=5}
  n8 = CameraPositionWS()
  n9 = FunctionInput(Preview=Normalize(VectorInput=Subtract(A=@n8, B=@n1))) {bUsePreviewValueAsDefault=true, description=Defaults to Camera Direction, inputName=RayDirection, sortPriority=4, previewValue=<0 0 0 256>}
  n10 = Subtract(A=@n1, B=@n2)
  n11 = Multiply(A=DotProduct(A=@n9, B=@n10)) {constB=2}
  n12 = Subtract(A=DotProduct(A=@n10, B=@n10), B=Multiply(A=@n5, B=@n5) {desc="^2"})
  n13 = Subtract(A=Multiply(A=@n11, B=@n11), B=Multiply(B=@n12) {constA=4}) {constB=4}
  n14 = Multiply(A=Subtract(A=SquareRoot(None=@n13), B=@n11)) {constB=0.5}
  FunctionOutput(None=@n14) {outputName="Intersection End Depth ", sortPriority=4}
  FunctionOutput(None=Add(A=Multiply(A=@n14, B=@n9), B=@n1)) {outputName=Intersection End World Position, sortPriority=3}
  n15 = Add(A=@n1, B=Multiply(A=@n14, B=@n9))
  FunctionOutput(None=@n15) {outputName="Ray Intersection Position "}
  n16 = Constant()
  FunctionOutput(None=If(A=@n13, B=@n16, "A > B"=Constant() {r=1}, "A < B"=@n16)) {outputName=Ray Hit Bool, sortPriority=2}
  FunctionOutput(None=Subtract(A=@n15, B=@n2)) {outputName="Projected Capture Vector ", sortPriority=6}
  FunctionOutput(None=Normalize(VectorInput=Subtract(A=@n15, B=@n2))) {outputName=Intersection Surface Normal, sortPriority=1}
  MaterialFunctionCall(X=Constant() {r=1}, Y=@n11, Z=@n12) {desc=QuadraticCoef, materialFunction=&MakeFloat3}
  OneMinus(None=@n6)
 ]
}
