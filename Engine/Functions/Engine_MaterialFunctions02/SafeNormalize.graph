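# SafeNormalize is a material function in the Misc and Math library that normalizes a vector without producing NaN at zero length.
# An If node compares the input's dot product with itself against zero, selecting either the normalized vector or the optional Default input, which falls back to its preview value.
# Each branch packs a zero-length flag into alpha, so one If feeds both the RGB result and the Length==0 output.
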MaterialFunction /Engine/Functions/Engine_MaterialFunctions02/SafeNormalize
graph {
 properties {
  bExposeToLibrary=true
  description="If the input vectors length == 0 then return 0 else return the normalized vector by default. This avoids returning NAN or NULL. Optionally, one can pass in a default vector that should be used when the vector "
  libraryCategoriesText [Misc,Math]
 }
 nodes [
  n1 = FunctionInput() {bUsePreviewValueAsDefault=true, inputName=Vector, previewValue=<0 0 1 1>}
  n2 = FunctionInput() {
   bUsePreviewValueAsDefault=true
   description=Vector to be used when the normalized vector reaches 0 in length.
   inputName=Default
   sortPriority=1
  }
  n3 = AppendVector(A=@n2, B=Constant() {r=1})
  n4 = If(A=DotProduct(A=@n1, B=@n1), B=Constant() {r=0}, "A > B"=AppendVector(A=Normalize(VectorInput=@n1), B=Constant()), "A == B"=@n3, "A < B"=@n3)
  FunctionOutput(None=ComponentMask(None=@n4) {b=true, g=true, r=true})
  FunctionOutput(None=ComponentMask(None=@n4) {a=true}) {description="Returns 1 if the vector's length was 0. Otherwise, this output returns 0.", outputName="Length==0", sortPriority=1}
 ]
}
