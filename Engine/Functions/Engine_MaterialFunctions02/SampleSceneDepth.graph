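# SampleSceneDepth is a library material function that reads scene depth at a position offset from the current screen pixel.
# Pixel Offset is scaled by scene texel size and Fraction Offset by view size, then each is added to viewport UV.
# Each offset UV feeds its own depth output, with the Use Custom Depth static switch choosing custom depth or regular scene depth for both.
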MaterialFunction /Engine/Functions/Engine_MaterialFunctions02/SampleSceneDepth
graph {
 properties {bExposeToLibrary=true, description="Samples the scene depth texture using multiple methods. ", libraryCategoriesText [Misc,Coordinates]}
 nodes [
  n1 = FunctionInput() {bUsePreviewValueAsDefault=true, inputName=Pixel Offset, inputType=FunctionInput_Vector2}
  n2 = ComponentMask(None=ScreenPosition().ViewportUV) {g=true, r=true}
  n3 = Add(A=Multiply(A=SceneTexelSize(), B=@n1), B=@n2)
  n4 = SceneTexture(UVs=@n3) {sceneTextureId=PPI_CustomDepth}
  n5 = FunctionInput(Preview=StaticBool()) {bUsePreviewValueAsDefault=true, inputName=Use Custom Depth, inputType=FunctionInput_StaticBool}
  FunctionOutput(None=StaticSwitch(True=ComponentMask(None=@n4.Color) {r=true}, False=SceneDepth(UVs=@n3).R, Value=@n5)) {outputName=Depth Using Pixel Offset}
  n6 = Add(
   A=Multiply(A=ViewSize(), B=FunctionInput(Preview=@n1) {bUsePreviewValueAsDefault=true, inputName=Fraction Offset, inputType=FunctionInput_Vector2})
   B=@n2
  )
  n7 = SceneTexture(UVs=@n6) {sceneTextureId=PPI_CustomDepth}
  FunctionOutput(None=StaticSwitch(True=ComponentMask(None=@n7.Color) {r=true}, False=SceneDepth(UVs=@n6).R, Value=@n5)) {outputName=Depth Using Fraction Offset}
 ]
}
