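# ScreenResolution is a library material function in the Misc and Coordinates categories that retrieves the visible screen resolution and the render target size.
# A ViewSize node feeds the Visible Resolution output.
# A ViewProperty node set to buffer size feeds the Buffer Resolution output, so materials can tell the visible view apart from the underlying render target.
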MaterialFunction /Engine/Functions/Engine_MaterialFunctions02/ScreenResolution
graph {
 properties {
  bExposeToLibrary=true
  description=Retrieves the visible screen resolution and the render target size.
  libraryCategoriesText [Misc,Coordinates]
 }
 nodes [
  FunctionOutput(None=ViewSize()) {outputName=Visible Resolution}
  FunctionOutput(None=ViewProperty() {property=MEVP_BufferSize}.Property) {outputName=Buffer Resolution, sortPriority=1}
 ]
}
