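# Material function SmoothStep, exposed to the library, that wraps the HLSL smoothstep function for smooth interpolation between two bounds.
# Three scalar inputs, Alpha, Min and Max, feed a Custom HLSL node whose single-float result is the function's only output.
# Each input uses its preview value as a default, so unconnected inputs fall back to a zero-to-one range with a midpoint Alpha.
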MaterialFunction /Engine/Functions/Engine_MaterialFunctions02/SmoothStep
graph {
 properties {bExposeToLibrary=true, description="HLSL smoothstep function. a custom node)"}
 nodes [
  MaterialExpressionFunctionInput_4 = FunctionInput() {bUsePreviewValueAsDefault=true, inputName=Alpha, inputType=FunctionInput_Scalar, previewValue=<0.5 0 0 1>}
  MaterialExpressionFunctionInput_2 = FunctionInput() {bUsePreviewValueAsDefault=true, inputName=Min, inputType=FunctionInput_Scalar, sortPriority=1}
  MaterialExpressionFunctionInput_3 = FunctionInput() {bUsePreviewValueAsDefault=true, inputName=Max, inputType=FunctionInput_Scalar, sortPriority=2, previewValue=<1 0 0 1>}
  n1 = Custom!(varA=@MaterialExpressionFunctionInput_4, varMin=@MaterialExpressionFunctionInput_2, varMax=@MaterialExpressionFunctionInput_3) {
   code="return smoothstep(varMin,varMax,varA);"
   outputType=CMOT_Float1
   inputs [
    {inputName=varA, input {expression=&self:MaterialExpressionFunctionInput_4}}
    {inputName=varMin, input {expression=&self:MaterialExpressionFunctionInput_2}}
    {inputName=varMax, input {expression=&self:MaterialExpressionFunctionInput_3}}
   ]
  }
  FunctionOutput(None=@n1)
 ]
}
