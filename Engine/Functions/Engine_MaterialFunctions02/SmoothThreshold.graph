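# SmoothThreshold is a library material function for image adjustment that ramps a scalar gradient's values below a threshold toward an adjustable level.
# It doubles Cutoff Value to form the threshold and passes gradient values above it through unchanged.
# Below it, a linear interpolation from Lerp Value to one, driven by gradient over threshold, is rescaled by the threshold to meet the original values continuously.
# The output is unclamped.
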MaterialFunction /Engine/Functions/Engine_MaterialFunctions02/SmoothThreshold
graph {
 properties {bExposeToLibrary=true, description=Smooth contrast applied after a threshold value is passed, libraryCategoriesText [Misc, Image Adjustment]}
 nodes [
  n1 = FunctionInput(Preview=ComponentMask(None=TextureCoordinate()) {g=true}) {
   bUsePreviewValueAsDefault=true
   description="Input a value to use as the base for the contrast operation. "
   inputName=Gradient
   inputType=FunctionInput_Scalar
  }
  n2 = FunctionInput(Preview=Constant()) {
   bUsePreviewValueAsDefault=true
   description=Values from the gradient input below this value will be effected by the operation.
   inputName=Cutoff Value
   inputType=FunctionInput_Scalar
  }
  n3 = Add(A=@n2, B=@n2)
  n4 = FunctionInput() {
   bUsePreviewValueAsDefault=true
   description=Add a negative or positive number to smoothly ramp to from the original values. Ramping occurs in values lower than the Cutoff Value.
   inputName=Lerp Value
   inputType=FunctionInput_Scalar
  }
  n5 = Constant() {r=1}
  n6 = LinearInterpolate(A=@n4, B=Constant() {r=1}, Alpha=Multiply(A=@n1, B=Divide(A=@n5, B=@n3)))
  FunctionOutput(None=If(A=@n1, B=@n3, "A > B"=@n1, "A < B"=Multiply(A=@n6, B=@n3))) {description=Be sure to clamp the values if needed. There is no cap on the outputed range.}
 ]
}
