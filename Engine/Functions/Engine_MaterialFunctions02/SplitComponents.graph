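# SplitComponents is an engine material function, exposed to the material library, that breaks a color input into its channels.
# A single Color input feeds an RGB passthrough output and three component masks that isolate red, green and blue.
# Each mask drives its own output, ordered R, G, B after RGB.
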MaterialFunction /Engine/Functions/Engine_MaterialFunctions02/SplitComponents
graph {
 properties {bExposeToLibrary=true}
 nodes [
  n1 = FunctionInput() {inputName=Color}
  FunctionOutput(None=@n1) {description=Passthrough, outputName=RGB}
  FunctionOutput(None=ComponentMask(None=@n1) {r=true}) {outputName=R, sortPriority=1}
  FunctionOutput(None=ComponentMask(None=@n1) {g=true}) {outputName=G, sortPriority=2}
  FunctionOutput(None=ComponentMask(None=@n1) {b=true}) {outputName=B, sortPriority=3}
 ]
}
