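# SteppingPannerTime is a library material function that outputs a stepped, time-driven offset to add to UVs, making noise maps look more random.
# It scales Time by TimeSpeed, adds Time Offset, then multiplies the fractional part by the two-component step count and rounds up.
# That result is divided, not multiplied, by Jump Amount Per Step, so the sequence repeats each scaled second.
# A fixed-rate Ceil quantization branch is left unconnected.
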MaterialFunction /Engine/Functions/Engine_MaterialFunctions02/SteppingPannerTime
graph {
 properties {
  bExposeToLibrary=true
  description="Add to UV coordinates to make them jump by the \"Jump Amount Per Step\" value for the number of times, per scaled second, set by the \"Number of Steps Per Unit of Time\". This will make noise maps seem to be more random than they are. "
  libraryCategoriesText [Misc,Texturing]
 }
 nodes [
  n1 = FunctionInput() {bUsePreviewValueAsDefault=true, inputName=TimeSpeed, inputType=FunctionInput_Scalar, previewValue=<1 0 0 1>}
  n2 = Add(
   A=Multiply(A=Time(), B=@n1)
   B=FunctionInput() {bUsePreviewValueAsDefault=true, inputName=Time Offset, inputType=FunctionInput_Scalar, sortPriority=100}
  )
  Divide(A=Ceil(None=Multiply(A=@n2) {constB=30})) {constB=30}
  n3 = FunctionInput() {bUsePreviewValueAsDefault=true, inputName=Number of Steps Per Unit of Time, inputType=FunctionInput_Vector2, previewValue=<30 30 0 1>}
  n4 = Multiply(A=Frac(None=@n2), B=@n3) {constB=30}
  n5 = FunctionInput() {bUsePreviewValueAsDefault=true, inputName=Jump Amount Per Step, inputType=FunctionInput_Vector2, previewValue=<6.14 3.0579 0 1>}
  FunctionOutput(None=Divide(A=Ceil(None=@n4), B=@n5))
 ]
}
