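# TangentBasis is a library-exposed material function that retrieves the tangent basis in world space.
# It passes each of the three unit axis vectors through its own Transform node, left at default spaces, then normalizes the result.
# The three normalized vectors are output separately as the X U, Y V and Z W axes.
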MaterialFunction /Engine/Functions/Engine_MaterialFunctions02/TangentBasis
graph {
 properties {bExposeToLibrary=true, description=Retrieves the tangent basis in world space}
 nodes [
  n1 = Constant3Vector() {constant=<rgba 1 0 0 0>}
  FunctionOutput(None=Normalize(VectorInput=Transform(None=@n1))) {outputName=X U}
  n2 = Constant3Vector() {constant=<rgba 0 1 0 0>}
  FunctionOutput(None=Normalize(VectorInput=Transform(None=@n2))) {outputName=Y V, sortPriority=1}
  n3 = Constant3Vector() {constant=<rgba 0 0 1 0>}
  FunctionOutput(None=Normalize(VectorInput=Transform(None=@n3))) {outputName=Z W, sortPriority=2}
 ]
}
