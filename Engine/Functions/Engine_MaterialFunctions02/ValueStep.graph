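# ValueStep is a library material function that turns a scalar gradient with values above one into a repeating black-and-white band mask.
# It adds Mask Offset Value to the gradient, then wraps the sum over a period of Number Before White Result plus one using divide, Frac and multiply.
# The wrapped value is rounded up with Ceil, and an If outputs white only where it equals Number Before White Result, otherwise black.
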MaterialFunction /Engine/Functions/Engine_MaterialFunctions02/ValueStep
graph {
 properties {
  bExposeToLibrary=true
  description=Takes a gradient with values over one and masks out a portion of the gradient based on input variables.
  libraryCategoriesText [Misc,Gradient]
 }
 nodes [
  n1 = FunctionInput(Preview=Multiply(
    A=MaterialFunctionCall() {materialFunction=&LinearGradient}.VGradient
    B=Constant() {r=3}
   )) {bUsePreviewValueAsDefault=true, description=Enter a gradient with values above one., inputName=Gradient, inputType=FunctionInput_Scalar}
  n2 = FunctionInput() {
   bUsePreviewValueAsDefault=true
   description=Offsets where the black spots are placed.
   inputName=Mask Offset Value
   inputType=FunctionInput_Scalar
   previewValue=<-1 0 0 1>
  }
  n3 = Add(A=@n1, B=@n2)
  n4 = FunctionInput() {
   bUsePreviewValueAsDefault=true
   description="If you enter a gradient that goes to 10 and this number is 9. 0-9 will be black and 10 will be white. "
   inputName=Number Before White Result
   inputType=FunctionInput_Scalar
   previewValue=<2 0 0 1>
  }
  n5 = Add(A=@n4, B=Constant() {r=1})
  n6 = Multiply(A=Frac(None=Divide(A=@n3, B=@n5) {desc=0-1}), B=@n5)
  n7 = Constant()
  FunctionOutput(None=If(A=Ceil(None=@n6), B=@n4, "A > B"=@n7, "A == B"=Constant() {r=1}, "A < B"=@n7))
 ]
}
