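# Material function that triplanar-projects a normal map from world position and outputs correctly oriented world-space normals, at higher instruction cost.
# World position over a negative divisor drives XZ, YZ and XY samples, each sign-flipped by the world normal's facing.
# Orthogonal-vector bases rotate them into world space, blended by overridable contrast-adjusted normal weights.
# Extra outputs give Z-only, XY-only and flat-top variants; some UV2 decode nodes are unconnected.
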MaterialFunction /Engine/Functions/Engine_MaterialFunctions02/WorldAlignedNormals_HighQuality
graph {
 properties {
  description="Applies normal maps to objects and correctly orients them. See WorldAlignedTexture_Complex for a more thorough implementation.  This version of world aligned normals is ~20 instructions more. The Normals are in worldspace by default. Transform them from world to tangent. "
  libraryCategoriesText [Misc,Texturing]
 }
 nodes [
  Constant() {r=1}
  n1 = FunctionInput() {
   bUsePreviewValueAsDefault=true
   description="All inputs need to be negative. For example: (-512,-256,-1024)"
   inputName=Negative World Position Divisor
   previewValue=<-255 -255 -255 1>
  }
  n2 = Divide(
   A=FunctionInput(Preview=WorldPosition() {worldPositionShaderOffset=WPT_ExcludeAllShaderOffsets}.XYZ) {bUsePreviewValueAsDefault=true, inputName=World Position, previewValue=<255 255 -255 1>}
   B=@n1
  ) {constB=-512}
  n3 = FunctionInput(Preview=TextureObject() {samplerType=SAMPLERTYPE_Normal, texture=/Engine/EngineMaterials/BaseFlattenNormalMap}) {inputName=Texture, inputType=FunctionInput_Texture2D}
  n4 = TextureSample(UVs=ComponentMask(None=@n2) {b=true, r=true}, Tex=@n3) {samplerSource=SSM_Wrap_WorldGroupSettings, samplerType=SAMPLERTYPE_Normal}
  n5 = FunctionInput(Preview=VertexNormalWS()) {
   bUsePreviewValueAsDefault=true
   description="All inputs need to be negative. For example: (-512,-256,-1024)"
   inputName=WorldNormal
   sortPriority=5
   previewValue=<-255 -255 -255 1>
  }
  n6 = Constant3Vector() {constant=<rgba 0 1 0 0>}
  n7 = MaterialFunctionCall(
   X=If(A=DotProduct(A=@n5, B=@n6) {desc=generate sign for this projection}, B=Constant(), "A > B"=Constant() {r=-1}, "A < B"=Constant() {r=1})
   Y=Constant() {r=-1}
   Z=Constant() {r=1}
  ) {materialFunction=&MakeFloat3}
  n8 = TextureSample(UVs=ComponentMask(None=@n2) {b=true, g=true}, Tex=@n3) {samplerSource=SSM_Wrap_WorldGroupSettings, samplerType=SAMPLERTYPE_Normal}
  n9 = Constant3Vector() {constant=<rgba 1 0 0 0>}
  n10 = MaterialFunctionCall(
   X=If(A=DotProduct(A=@n5, B=@n9) {desc=generate sign for this projection}, B=Constant(), "A > B"=Constant() {r=1}, "A < B"=Constant() {r=-1})
   Y=Constant() {r=-1}
   Z=Constant() {r=1}
  ) {materialFunction=&MakeFloat3}
  n11 = MaterialFunctionCall(Float3=Abs(None=@n5)) {materialFunction=&BreakOutFloat3Components}
  n12 = Constant() {r=2}
  n13 = MaterialFunctionCall(In=@n11.R, Contrast=@n12) {materialFunction=&CheapContrast}
  n14 = MaterialFunctionCall!(Vector1=@n5, Vector2=Constant3Vector() {constant=<rgba 0 0 1 0>}) {materialFunction=/Engine/Functions/Engine_MaterialFunctions02/Math/CreateThirdOrthogonalVector}
  n15 = MaterialFunctionCall(
   VectorToTransform=LinearInterpolate(
    A=Multiply(A=@n4.RGB, B=@n7.Result)
    B=Multiply(A=@n8.RGB, B=@n10.Result)
    Alpha=FunctionInput(Preview=@n13.Result) {bUsePreviewValueAsDefault=true, inputName=X lerp alpha, inputType=FunctionInput_Scalar, sortPriority=10}
   )
   BasisX=@n14.Vector3
   BasisY=@n14.Vector3
   BasisZ=@n14.Vector3
  ) {
   desc=lerps in when the vertex normal faces X or Negative X
   materialFunction=&Transform3x3Matrix
  }
  n16 = ComponentMask(None=@n2) {g=true, r=true}
  n17 = If(
   A=DotProduct(A=@n5, B=Constant3Vector() {constant=<rgba 0 0 1 0>}) {desc=generate sign for this projection}
   B=Constant()
   "A > B"=Constant() {r=1}
   "A < B"=Constant() {r=-1}
  )
  n18 = Constant() {r=-1}
  n19 = Constant() {r=1}
  n20 = MaterialFunctionCall(Vector1=@n5, Vector2=Constant3Vector() {constant=<rgba 0 1 0 0>}) {materialFunction=&CreateThirdOrthogonalVector}
  n21 = MaterialFunctionCall(
   VectorToTransform=Multiply(
    A=TextureSample(UVs=@n16, Tex=@n3) {samplerSource=SSM_Wrap_WorldGroupSettings, samplerType=SAMPLERTYPE_Normal}.RGB
    B=MaterialFunctionCall(X=@n17, Y=@n18, Z=@n19) {materialFunction=&MakeFloat3}.Result
   )
   BasisX=@n20.Vector3
   BasisY=@n20.Vector3
   BasisZ=@n20.Vector3
  ) {
   desc=lerps in when the vertex normal faces X or Negative X
   materialFunction=&Transform3x3Matrix
  }
  n22 = FunctionInput(Preview=MaterialFunctionCall(In=@n11.B, Contrast=@n12) {materialFunction=&CheapContrast}.Result) {bUsePreviewValueAsDefault=true, inputName=Z lerp Alpha, inputType=FunctionInput_Scalar, sortPriority=2}
  FunctionOutput(None=LinearInterpolate(A=@n15.Transformed Vector, B=@n21.Transformed Vector, Alpha=@n22)) {outputName=World Space Normals XYZ}
  FunctionOutput(None=@n21.Transformed Vector) {outputName=Z Projection World Space Normals, sortPriority=3}
  FunctionOutput(None=@n15.Transformed Vector) {outputName=XY Projection World Space Normals, sortPriority=2}
  FunctionOutput(None=LinearInterpolate(A=@n15.Transformed Vector, B=@n5, Alpha=@n22)) {outputName=XY Flat Top Projection World Space Normals, sortPriority=1}
  n23 = TextureCoordinate() {coordinateIndex=2}
  n24 = MaterialFunctionCall(Float2=@n23) {materialFunction=&BreakOutFloat2Components}
  MaterialFunctionCall(
   X=@n24.R
   Y=@n24.R
   Z=MaterialFunctionCall(Float2=TextureCoordinate() {coordinateIndex=2}) {materialFunction=&BreakOutFloat2Components}.R
  ) {materialFunction=&MakeFloat3}
  n25 = Constant()
  n26 = MaterialFunctionCall(Float2=@n23) {materialFunction=&BreakOutFloat2Components}
  n27 = Constant()
  n28 = Constant() {r=1}
  n29 = Constant() {r=-1}
  n30 = ConstantBiasScale(None=Abs(None=@n26.R)) {bias=-0.5, scale=2}
  Multiply(
   A=MaterialFunctionCall(X=@n25, Y=@n25, Z=If(A=@n26.R, B=@n27, "A > B"=@n28, "A == B"=@n28, "A < B"=@n29)) {materialFunction=&MakeFloat3}.Result
   B=DeriveNormalZ(InXY=AppendVector(A=@n30, B=@n26.G))
  )
 ]
}
