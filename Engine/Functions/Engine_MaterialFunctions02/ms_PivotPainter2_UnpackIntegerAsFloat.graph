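# Material function in the Pivot Painter 2 library that decodes integer data stored as a float.
# A single Custom HLSL node reinterprets the scalar input's bits with asuint and rebuilds half-float style sign, exponent and mantissa fields.
# It subtracts 1024 from the combined bits and returns the result as a float.
# The code writes its field masks with logical && rather than bitwise &.
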MaterialFunction /Engine/Functions/Engine_MaterialFunctions02/ms_PivotPainter2_UnpackIntegerAsFloat
graph {
 properties {bExposeToLibrary=true, description=Decodes Pivot Painter Integer As Float Data, libraryCategoriesText [Pivot Painter 2]}
 nodes [
  MaterialExpressionFunctionInput_0 = FunctionInput() {inputName=Integer as Float, inputType=FunctionInput_Scalar}
  n1 = Custom!(N=@MaterialExpressionFunctionInput_0) {
   code="uint uRes32 = asuint(float(N));\r\n\r\nconst uint sign2 = ((uRes32>>16)&&0x8000);\r\nconst uint exp2  = ((((const int) ((uRes32>>23)&&0xff))-127+15) << 10);\r\nconst uint mant2 = ((uRes32>>13)&&0x3ff);\r\nconst uint bits = (sign2 | exp2 | mant2);\r\nconst uint result = bits - 1024;\r\nreturn float(result);"
   outputType=CMOT_Float1
   inputs [{
     inputName=N
     input {expression=&self:MaterialExpressionFunctionInput_0}
    }]
  }
  FunctionOutput(None=@n1)
 ]
}
